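# Niagara system for the CattleJar pickaxe swing trails, built from two ribbon emitters that follow the weapon's skeletal mesh.
# BladeTrailMain and SharpBladeTrail spawn anim-trail particles between different bone or socket pairs, with opposite camera offsets.
# Ribbon lifetime ramps up quickly with system age, avoiding full-length trails at activation.
# Alpha fades over curves scaled by swing fade and time-of-day values, RGB takes time-of-day tint, and the sharp trail also narrows.
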NiagaraSystem /BRCosmetics/Weapons/FORT_Melee/Pickaxe_CattleJar/FX/NS_Pickaxe_CattleJar_Trails
dependencies {
 usedDataInterfaces [&NiagaraDataInterfaceCurve, &NiagaraDataInterfaceSkeletalMesh, &NiagaraDataInterfaceColorCurve]
 usedDynamicInputs [&ColorFromCurve]
 usedModules [
  &self:SystemSpawnScript.SystemScriptSource.SystemScriptGraph.NiagaraNodeAssignment_0.SetVariables_#1
  &SystemState
  &self:SystemSpawnScript.SystemScriptSource.SystemScriptGraph.NiagaraNodeAssignment_1.SetVariables_#2
  &self:SystemSpawnScript.SystemScriptSource.SystemScriptGraph.NiagaraNodeAssignment_2.SetVariables_#3
 ]
}
emitters {
 BladeTrailMain {
  emitterUpdateScript [
   &EmitterState()
   &FN_SpawnAnimTrail(Spawn Interval=0)
  ]
  particleSpawnScript [
   &FN_InitializeAnimTrailBoneIndecies(First Bone or Socket Index=1, Second Bone or Socket Index=0)
   &FN_InitializeAnimTrail(
    Color=<rgba 0.67 0.7525 1 0.55>
    Lifetime=&FloatFromCurve(
     FloatCurve=DataInterface() {propertyValues {Curve {keys [<at 0 0 RCIM_Cubic>, {arriveTangent=0.0455, interpMode=RCIM_Cubic, leaveTangent=0.0455, tangentMode=RCTM_User, time=0.04, value=1}]}}}
     CurveIndex=$System.Age
     Scale Curve=0.2
    )
    SkeletalMesh=$System.Skeletal Mesh
   )
   &FN_InitializeAnimTrail(
    Color=&FNTimeOfDayColor(
     Night Color=<rgba 0.129 0.2042 0.43 1>
     Morning Color=<rgba 0.5 0.39 0.3992 1>
     Day Color=<rgba 0.55 0.495 0.4675 1>
     Evening Color=<rgba 0.46 0.3565 0.322 1>
    )
    Lifetime=0.13
    SkeletalMesh=$System.Skeletal Mesh
   ) {enabled=false, module=FN_InitializeAnimTrail001}
   &CameraOffset(Camera Offset Amount=4)
   &AddVelocity(Velocity=<1000 0 0>) {enabled=false}
  ]
  particleUpdateScript [
   &SolveForcesAndVelocity(Clamp Velocity=false, Limit Acceleration=false)
   &FN_UpdateAnimTrail(Locked Particle RibbonWidth=1, SkeletalMesh=$System.Skeletal Mesh) {module=FN_UpdateAnimTrail001}
   &ScaleColor(
    Scale Alpha=&FloatFromCurve(
     FloatCurve=DataInterface() {propertyValues {Curve {keys [
         {arriveTangent=-2.003, interpMode=RCIM_Cubic, leaveTangent=-2.003, tangentMode=RCTM_User, time=0.2, value=1}
         {arriveTangent=-0.2408, interpMode=RCIM_Cubic, leaveTangent=-0.2408, tangentMode=RCTM_User, time=0.7, value=0.05}
         <at 1 0 RCIM_Cubic>
        ]}}}
     Scale Curve=&Multiply_Float(A=$System.SwingGeneralFade, B=$System.TODFloat)
    )
    Scale RGB=&MakeVectorFromLinearColorRGB(Color=$System.TODColor)
    ScaleA=true
    ScaleRGB=true
   )
  ]
  rendererClasses [&NiagaraRibbonRendererProperties]
  renderers [{rendererClass=&NiagaraRibbonRendererProperties}]
 }
 SharpBladeTrail {
  emitterUpdateScript [
   &EmitterState()
   &FN_SpawnAnimTrail(Spawn Interval=0)
  ]
  particleSpawnScript [
   &FN_InitializeAnimTrailBoneIndecies(First Bone or Socket Index=2, Second Bone or Socket Index=0)
   &FN_InitializeAnimTrail(
    Color=<rgba 0.92 0.94 1 1>
    Lifetime=&FloatFromCurve(
     FloatCurve=DataInterface() {propertyValues {Curve {keys [<at 0 0 RCIM_Cubic>, {arriveTangent=0.0455, interpMode=RCIM_Cubic, leaveTangent=0.0455, tangentMode=RCTM_User, time=0.04, value=1}]}}}
     CurveIndex=$System.Age
     Scale Curve=0.18
    )
    SkeletalMesh=$System.Skeletal Mesh
   )
   &CameraOffset(Camera Offset Amount=-4)
   &AddVelocity(Velocity=<1000 0 0>) {enabled=false}
  ]
  particleUpdateScript [
   &FN_UpdateAnimTrail(SkeletalMesh=$System.Skeletal Mesh)
   &SolveForcesAndVelocity(Clamp Velocity=false, Limit Acceleration=false)
   &ScaleRibbonWidth(Ribbon Width Scale=&FloatFromCurve(FloatCurve=DataInterface() {propertyValues {Curve {keys [
         {arriveTangent=-1.6883, interpMode=RCIM_Cubic, leaveTangent=-1.6883, tangentMode=RCTM_User, time=0, value=1}
         {interpMode=RCIM_Cubic, tangentMode=RCTM_User, time=1, value=0.1}
        ]}}}))
   &ScaleColor(
    Scale Alpha=&FloatFromCurve(
     FloatCurve=DataInterface() {propertyValues {Curve {keys [
         {arriveTangent=-1.8064, interpMode=RCIM_Cubic, leaveTangent=-1.8064, tangentMode=RCTM_User, time=0, value=1}
         {interpMode=RCIM_Cubic, tangentMode=RCTM_User, time=1, value=0}
        ]}}}
     Scale Curve=&Multiply_Float(
      A=&FNTimeOfDayScalar(Night Scale=1.5, Morning Scale=1.2, Day Scale=2, Evening Scale=1)
      B=$System.SwingGeneralFade
     )
    )
    Scale RGB=&MakeVectorFromLinearColorRGB(Color=$System.TODColor)
    ScaleA=true
    ScaleRGB=true
   )
  ]
  rendererClasses [&NiagaraRibbonRendererProperties]
  renderers [{rendererClass=&NiagaraRibbonRendererProperties}]
 }
}
summary {
 systemName=NS_Pickaxe_CattleJar_Trails
 emitters [
  {emitterName=BladeTrailMain, rendererClasses [&NiagaraRibbonRendererProperties]}
  {emitterName=SharpBladeTrail, rendererClasses [&NiagaraRibbonRendererProperties]}
 ]
 userVariables [
  User.bisSwinging {type=NiagaraBool}
  User.NewVector3f {type=Vector3f, description="", defaultValue {struct=Vector3f, value=<0 0 0>}}
  User.RibbonIDOffset {type=NiagaraInt32}
  User.Variant {type=NiagaraFloat, defaultValue=0, description=""}
 ]
}
