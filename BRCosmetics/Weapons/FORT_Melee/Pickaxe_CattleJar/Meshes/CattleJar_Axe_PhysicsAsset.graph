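# Physics asset for the CattleJar pickaxe, a two-handed melee cosmetic weapon mesh.
# It holds a single capsule body on the Melee_TwoHanded_Handle bone, offset up along the handle and slightly tilted.
# It defines no constraints, so the weapon has one simple collision shape with no jointed parts.
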PhysicsAsset /BRCosmetics/Weapons/FORT_Melee/Pickaxe_CattleJar/Meshes/CattleJar_Axe_PhysicsAsset
bodies {Melee_TwoHanded_Handle [{shapeType=Capsule, length=89.7167, radius=30.6163, center=<0.8449 4.6992 49.277>, rotation=<rot -5.2481 -90 -0.0597>}]}
constraints []
